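# Physics asset for the ScarC vehicle body cosmetic mesh, likely the battle-royale variant judging by its BR suffix.
# It holds a single body on the root bone with no collision shapes listed.
# It has no constraints, so it defines no articulated or simulated joints.
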PhysicsAsset /VehicleCosmetics/Bodies/ScarC/Meshes/Body_ScarC_PhysicsBR
bodies {root []}
constraints []
